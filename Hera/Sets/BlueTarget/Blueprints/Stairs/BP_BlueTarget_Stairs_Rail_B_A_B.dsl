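; Blueprint for a stair rail piece in the BlueTarget set, variant Rail_B_A_B, derived from a parent rail blueprint.
; Its construction script runs the parent's construction, then calls SetColoronSecondaryMesh to tint the SM_BlueTarget_Stairs_RailAddon_A mesh a fixed muted blue.
; BeginPlay, overlap and tick events are empty, so it carries no runtime logic of its own.
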(event EventBeginPlay)

(event Collision|EventActorBeginOverlap
  (params (OtherActor :type Actor%20Object%20Reference)))

(event EventTick
  (params (DeltaSeconds :type Float%20%28single-precision%29)))

(fn ConstructionScript
  (|CallParentFunction)
  (Default|SetColoronSecondaryMesh "(R=0.045186,G=0.111932,B=0.238398,A=1.000000)" (Variables|Default|GetSM_BlueTarget_Stairs_RailAddon_A)))
